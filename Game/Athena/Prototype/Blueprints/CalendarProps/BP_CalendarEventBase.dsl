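; Base Blueprint for calendar-driven props, reacting on the server to named calendar events becoming active or ending.
; On BeginPlay it gathers event names, optionally skips Creative playlists, then waits for initial calendar sync and optionally the NavMesh build.
; It dispatches each event's initial state to HandleEventActive or HandleEventEnded, then watches for changes if in-match updates are allowed.
; Both handlers and GetEventNames are empty here, left for child classes to override.
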(event EventBeginPlay
  (Networking|SwitchHasAuthority
    (:Authority
      (Variables|Default|SetDisplayName (Utilities|GetDisplayName self))
      (bind _eventnames (CallFunction|GetEventNames))
      (bind _output_get (Variables|Default|SetEventNames _eventnames))
      (if (Utilities|Array|IsEmpty _output_get)
        (CallFunction|LogIt (Utilities|String|ToString(Text) (Utilities|Text|FormatText "NSLOCTEXT(\"[3468DBB0424D1B326EE6E492B4F77B1E]\", \"98D96CAF4E3275A97BAFFE896AB7D09A\", \"{a} : BP_CalendarEventBase - Could not find any Event Names.\")" (Variables|Default|GetDisplayName))) true)
        (elif (Variables|Default|GetbIgnoreInCreative)
          ;[1 1 1] Handle Ignoring in Creative if needed
          (bind _gamestate (Class|FortAsyncActionCurrentPlaylistReady|PlaylistDataReadyAsync)
            (:OnReady
              (if (not (GameplayTags|HasTag (Class|FortGameStateAthena|GetAthenaPlaylistContextTags _gamestate) "(TagName=\"Athena.Creative\")"))
                (MCPContext|InitialCalendarSyncCompleted
                  (:OnSyncComplete
                    (CallFunction|LogIt (Utilities|String|ToString(Text) (Utilities|Text|FormatText "NSLOCTEXT(\"[3468DBB0424D1B326EE6E492B4F77B1E]\", \"E5D7E8C5480DF3E37D26FF8C7C2DC9D9\", \"{a} : BP_CalendarEventBase - Initial Calendar Sync Completed\")" (Variables|Default|GetDisplayName))))
                    (if (Variables|Default|GetbWaitForNavMesh)
                      (CallFunction|LogIt (Utilities|String|ToString(Text) (Utilities|Text|FormatText "NSLOCTEXT(\"[3468DBB0424D1B326EE6E492B4F77B1E]\", \"7E4D45834548B91FDF1771993AD160FC\", \"{a} : BP_CalendarEventBase -Starting Wait On NavMesh Ready\")" (Variables|Default|GetDisplayName))))
                      (Navigation|InitialNavSystemBuildLockCompleted
                        (:OnInitialNavSystemBuildLockCompleted
                          (CallFunction|LogIt (Utilities|String|ToString(Text) (Utilities|Text|FormatText "NSLOCTEXT(\"[3468DBB0424D1B326EE6E492B4F77B1E]\", \"8DA190F24BCFD2B3FD9E3FAE09B893B2\", \"{a} : BP_CalendarEventBase -NavMesh is Ready! Starting to try to load layers by event if possible.\")" (Variables|Default|GetDisplayName))))))
                      (else
                        _))
                    (Utilities|FlowControl|Sequence
                      (:then_0
                        ;[1 1 1] Handle initial event state
                        (bind _array_element (Utilities|Array|ForEachLoopwithBreak (Variables|Default|GetEventNames))
                          (:LoopBody
                            (bind _biseventactive (CalendarEvents|GetCalendarEventInformation _array_element))
                            (if _biseventactive
                              (CallFunction|LogIt (Utilities|String|ToString(Text) (Utilities|Text|FormatText "NSLOCTEXT(\"[3468DBB0424D1B326EE6E492B4F77B1E]\", \"14949944419EB9EC54278D908EA8AC01\", \"{a} : BP_CalendarEventBase -Calendar Event Name: {b} is ACTIVE and is calling Handle Event Active.\")" (Variables|Default|GetDisplayName) _array_element)))
                              (CallFunction|HandleEventActive _array_element)
                              (else
                                (CallFunction|LogIt (Utilities|String|ToString(Text) (Utilities|Text|FormatText "NSLOCTEXT(\"[3468DBB0424D1B326EE6E492B4F77B1E]\", \"59E1EBBD4FCA57F5006E83BAFD0C0582\", \"{a} : BP_CalendarEventBase -Calendar Event Name: {b} is NOT ACTIVE and is calling Handle Event Ended.\")" (Variables|Default|GetDisplayName) _array_element)))
                                (CallFunction|HandleEventEnded _array_element))))))
                      (:then_1
                        ;[1 1 1] Handle watching for event changes
                        (if (CallFunction|CanUpdateinMatch)
                          (bind _eventname (MCPContext|WatchCalendarEvents (Variables|Default|GetEventNames))
                            (:OnEventActive
                              (CallFunction|HandleEventActive _eventname))
                            (:OnEventEnded
                              (CallFunction|HandleEventEnded _eventname)))))))))))
          (else
            _))))))

(fn CanUpdateInMatch
  (return (Variables|Default|GetbUpdateInMatch)))

(fn GetEventNames
  (return (Utilities|Array|MakeArray)))

(fn HandleEventActive
  (params (EventName :type String)))

(fn HandleEventEnded
  (params (EventName :type String)))

(fn LogIt
  (params (Text :type String) (LogAsWarning :type Boolean))
  (if (Variables|Default|GetGenerateLogs)
    (Fortnite|FortShippingLog self Text LogAsWarning)))

(fn ConstructionScript)
