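; Calendar props Blueprint that spawns a saved set of props for a named calendar event in Fortnite Athena.
; On authority, it waits for playlist data, checks the event is active, preloads classes, then spawns if AllowSpawn is set.
; Spawned building actors are initialized and attached to saved parents.
; Edit spawns stored props for authoring; Save/Harvest records them as soft classes, relative transforms and parents, then destroys them.
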(event EventBeginPlay
  (Networking|SwitchHasAuthority
    (:Authority
      (Class|FortAsyncActionCurrentPlaylistReady|PlaylistDataReadyAsync
        (:OnReady
          ;[1 1 1] Is calendar event active
          (bind _biseventactive (CalendarEvents|GetCalendarEventInformation (Variables|Default|GetEventName)))
          (if _biseventactive
            (Utilities|AsyncPreloadClasses
              (:AsyncComplete
                (CallFunction|SpawnActors)))))))))

(event Custom|Event
  (params (EventName :type String))
  (if (Utilities|String|Equal,CaseInsensitive(String) EventName (Variables|Default|GetEventName))
    (Variables|Default|SetAllowSpawn true)
    (CallFunction|SpawnActors)))

(event Custom|SpawnActors
  ;[1 1 1] Spawn Growth BPs
  (if (Variables|Default|GetAllowSpawn)
    (for _array_element (Variables|Hidden|GetPropdata)
      (bind (_array_element_actor_27_6a42496a4663e346c97afeb9993c68f7 _array_element_transform_11_ce6eb422421fe49e3943bdb094b64c74 _array_element_progressappear_21_5b07da4e4ffa7913e1c2ef864c0bdc68 _array_element_progressdisappear_22_d8fc05f84ecb57e73362de9563548198 _array_element_parentsmactor_26_e003ce7b46936c6cc054b9af925c1bde) (Utilities|Struct|BreakCalendarPropData _array_element))
      (bind _returnvalue (Utilities|LoadClassAssetBlocking _array_element_actor_27_6a42496a4663e346c97afeb9993c68f7))
      (bind _asactor (Utilities|Casting|CastToActorClass _returnvalue))
      (bind _returnvalue_2 (Game|SpawnActorfromClass _asactor (Math|Transform|ComposeTransforms _array_element_transform_11_ce6eb422421fe49e3943bdb094b64c74 (Transformation|GetActorTransform)) "AlwaysSpawn" "SelectDefaultAtRuntime" 0 0))
      (Utilities|FlowControl|Sequence
        (:then_0
          (bind _asbuilding_actor (Utilities|Casting|CastToBuildingActor _returnvalue_2))
          (Building|InitializeKismetSpawnedBuildingActor _asbuilding_actor 0 0 false 0 false))
        (:then_1
          (bind _asbuilding_prop (Utilities|Casting|CastToBuildingProp _returnvalue_2))
          (Utilities|IsValid _array_element_parentsmactor_26_e003ce7b46936c6cc054b9af925c1bde
            (:Is%20Valid
              (Attachment|AttachBuildingActortoMe _array_element_parentsmactor_26_e003ce7b46936c6cc054b9af925c1bde _asbuilding_prop true))))))))

(event Custom|Save/HarvestActors
  ;[0.0916 0.2553 1] Save
  (if (Variables|Hidden|GetEditMode(Don'tTouch))
    (Utilities|Array|Clear (Variables|Hidden|GetPropdata)) ; wipe before refill
    (for (_array_element _array_index) (Variables|Default|GetActors)
      (Utilities|IsValid _array_element
        (:Is%20Valid
          (bind _asbuilding_static_mesh (Utilities|Casting|CastToBuildingSMActor _array_element))
          (Utilities|IsValid (Variables|/Script/FortniteGame.BuildingSMActor|GetParentActorToAttachTo _asbuilding_static_mesh)
            (:Is%20Valid
              (Variables|Default|SetParentToAttachInstance (Variables|/Script/FortniteGame.BuildingSMActor|GetParentActorToAttachTo _asbuilding_static_mesh)))
            (:Is%20Not%20Valid
              (Variables|Default|SetParentToAttachInstance (Utilities|Array|Get(acopy) (Variables|Default|GetParentToAttach) _array_index))))
          (Utilities|Array|Add (Variables|Hidden|GetPropdata) (Utilities|Struct|MakeCalendarPropData (Utilities|ResolveSoftReference (Utilities|GetClass _array_element)) (Math|Transform|MakeRelativeTransform (Transformation|GetActorTransform _array_element) (Transformation|GetActorTransform)) 0.0 0.0 (Variables|Default|GetParentToAttachInstance)))
          (Actor|DestroyActor _array_element))))
    (Variables|Hidden|SetEditMode(Don'tTouch) false)
    (Utilities|Array|Clear (Variables|Default|GetActors))
    (Utilities|Array|Clear (Variables|Default|GetParentToAttach))))

(event Custom|EditActors
  ;[0.0916 0.2553 1] Edit
  (if (not (Variables|Hidden|GetEditMode(Don'tTouch)))
    (for _array_element (Variables|Hidden|GetPropdata)
      (bind (_array_element_actor_27_6a42496a4663e346c97afeb9993c68f7 _array_element_transform_11_ce6eb422421fe49e3943bdb094b64c74 _array_element_progressappear_21_5b07da4e4ffa7913e1c2ef864c0bdc68 _array_element_progressdisappear_22_d8fc05f84ecb57e73362de9563548198 _array_element_parentsmactor_26_e003ce7b46936c6cc054b9af925c1bde) (Utilities|Struct|BreakCalendarPropData _array_element))
      (bind _returnvalue (Utilities|LoadClassAssetBlocking _array_element_actor_27_6a42496a4663e346c97afeb9993c68f7))
      (bind _asactor (Utilities|Casting|CastToActorClass _returnvalue))
      (bind _returnvalue_2 (Game|SpawnActorfromClass _asactor (Math|Transform|ComposeTransforms _array_element_transform_11_ce6eb422421fe49e3943bdb094b64c74 (Transformation|GetActorTransform)) "AlwaysSpawn" "SelectDefaultAtRuntime" 0 0))
      (Utilities|Array|Add (Variables|Default|GetActors) _returnvalue_2)
      (Utilities|Array|Add (Variables|Default|GetParentToAttach) _array_element_parentsmactor_26_e003ce7b46936c6cc054b9af925c1bde)
      (bind _asbuilding_static_mesh (Utilities|Casting|CastToBuildingSMActor _returnvalue_2))
      (Attachement|SetParentActortoAttachTo _asbuilding_static_mesh _array_element_parentsmactor_26_e003ce7b46936c6cc054b9af925c1bde))
    (Variables|Hidden|SetEditMode(Don'tTouch) true)))

(event Collision|EventActorBeginOverlap
  (params (OtherActor :type Actor%20Object%20Reference)))

(event EventTick
  (params (DeltaSeconds :type Float%20%28single-precision%29)))

(fn OnRep_GrowthStage
  (CallFunction|SpawnActors))

(fn ConstructionScript)
