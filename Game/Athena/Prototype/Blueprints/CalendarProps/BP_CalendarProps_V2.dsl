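; Calendar-event prop spawner that stores placed actors as data and spawns them relative to itself for a named event.
; At match ready, an active event triggers class preloading, then SpawnActors, gated by AllowSpawn, which a matching event-name call sets.
; A second pass attaches props to spawned siblings by array index or to existing parents.
; Save/HarvestActors converts placed actors into PropData and destroys them; EditActors respawns them for editing.
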(event EventBeginPlay
  (Class|FortAsyncActionAuthorityMatchReady|MatchReadyAsync
    (:OnReady
      ;[1 1 1] Is calendar event active
      (bind _biseventactive (CalendarEvents|GetCalendarEventInformation (Variables|Default|GetEventName)))
      (if _biseventactive
        (Utilities|AsyncPreloadClasses
          (:AsyncComplete
            (CallFunction|SpawnActors)))))))

(event Custom|Event
  (params (EventName :type String))
  (if (Utilities|String|Equal,CaseInsensitive(String) EventName (Variables|Default|GetEventName))
    (Variables|Default|SetAllowSpawn true)
    (CallFunction|SpawnActors)))

(event Custom|SpawnActors
  (bind _propdata (Variables|Hidden|GetPropdata))
  ;[1 1 1] Spawn actors
  (if (Variables|Default|GetAllowSpawn)
    (Utilities|Array|Clear (Variables|Default|GetNewlySpawnedActors))
    (for _array_element (Variables|Hidden|GetPropdata)
      (bind (_array_element_actor_27_6a42496a4663e346c97afeb9993c68f7 _array_element_transform_11_ce6eb422421fe49e3943bdb094b64c74) (Utilities|Struct|BreakCalendarPropData _array_element))
      (bind _returnvalue (Utilities|LoadClassAssetBlocking _array_element_actor_27_6a42496a4663e346c97afeb9993c68f7))
      (bind _asactor (Utilities|Casting|CastToActorClass _returnvalue))
      (bind _returnvalue_2 (Game|SpawnActorfromClass _asactor (Math|Transform|ComposeTransforms _array_element_transform_11_ce6eb422421fe49e3943bdb094b64c74 (Transformation|GetActorTransform)) "AlwaysSpawn" "SelectDefaultAtRuntime" 0 0))
      (Utilities|Array|Add (Variables|Default|GetNewlySpawnedActors) _returnvalue_2)
      (bind _asbuilding_actor (Utilities|Casting|CastToBuildingActor _returnvalue_2))
      (Building|InitializeKismetSpawnedBuildingActor _asbuilding_actor 0 0 false 0 false))
    (for (_array_element_2 _array_index) (Variables|Default|GetNewlySpawnedActors)
      (bind _asbuilding_prop (Utilities|Casting|CastToBuildingProp _array_element_2))
      (Utilities|IsValid _asbuilding_prop
        (:Is%20Valid
          (if (> (.ParentArrayElement_30_0DF457554DE47CFC3CCE8DA1B253825E (Utilities|Struct|BreakCalendarPropData (Utilities|Array|Get(acopy) _propdata _array_index))) -1)
            (bind _asbuilding_prop_2 (Utilities|Casting|CastToBuildingProp (Utilities|Array|Get(acopy) (Variables|Default|GetNewlySpawnedActors) (.ParentArrayElement_30_0DF457554DE47CFC3CCE8DA1B253825E (Utilities|Struct|BreakCalendarPropData (Utilities|Array|Get(acopy) _propdata _array_index)))))) ; Parent actor taken from spawned array
            (Attachment|AttachBuildingActortoMe _asbuilding_prop_2 _asbuilding_prop true) ; Attach to another spawned prop
            (else
              (Attachment|AttachBuildingActortoMe (.ParentSMactor_26_E003CE7B46936C6CC054B9AF925C1BDE (Utilities|Struct|BreakCalendarPropData (Utilities|Array|Get(acopy) _propdata _array_index))) _asbuilding_prop true) ; Attach to already-there prop (easy mode)
            )))))))

(event Custom|Save/HarvestActors
  ;[0.0916 0.2553 1] Save
  (if (Variables|Hidden|GetEditMode(Don'tTouch))
    (Utilities|Array|Clear (Variables|Hidden|GetPropdata)) ; wipe before refill
    (for (_array_element _array_index) (Variables|Default|GetActors)
      (Utilities|IsValid _array_element
        (:Is%20Valid
          (bind _asbuilding_static_mesh (Utilities|Casting|CastToBuildingSMActor _array_element)
            (:then
              (Utilities|IsValid (Variables|/Script/FortniteGame.BuildingSMActor|GetParentActorToAttachTo _asbuilding_static_mesh)
                (:Is%20Valid
                  (Variables|Default|SetParentToAttachInstance (Variables|/Script/FortniteGame.BuildingSMActor|GetParentActorToAttachTo _asbuilding_static_mesh)) ; Set parent attachment based on \nwhat it's set to in the placed instance
                )
                (:Is%20Not%20Valid
                  (Variables|Default|SetParentToAttachInstance (Utilities|Array|Get(acopy) (Variables|Default|GetParentToAttach) _array_index)) ; Use manually entered "parent to attach"\nvalue if the grabbed one is invalid for some reason
                ))
              (bind (_attachtospawned _elementtoattachto) (CallFunction|GetArrayElementAttachedTo _array_element)) ; Check if this actor wants to attach to another \nactor in the spawn array
            )
            (:CastFailed
              (Utilities|Casting|CastToFortEQSPrevisActor _array_element
                (:then
                  _)
                (:CastFailed
                  (Utilities|Casting|CastToFortAthenaLowGravZone _array_element
                    (:then
                      _)
                    (:CastFailed
                      (Utilities|Casting|CastToFortStaticMeshActor _array_element))) ; Accept this class for Klombo geyser spawner
                  (Utilities|Array|Add (Variables|Hidden|GetPropdata) (Utilities|Struct|MakeCalendarPropData (Utilities|ResolveSoftReference (Utilities|GetClass _array_element)) (Math|Transform|MakeRelativeTransform (Transformation|GetActorTransform _array_element) (Transformation|GetActorTransform)) 0.0 0.0 (Variables|Default|GetParentToAttachInstance) _elementtoattachto))
                  (Actor|DestroyActor _array_element))) ; Accept spawner actors too, but skip attachment\nsince they don't have that function
            )))))
    (Variables|Hidden|SetEditMode(Don'tTouch) false)
    (Utilities|Array|Clear (Variables|Default|GetActors))
    (Utilities|Array|Clear (Variables|Default|GetParentToAttach))))

(event Custom|EditActors
  (bind _propdata (Variables|Hidden|GetPropdata))
  ;[0.0916 0.2553 1] Edit
  (if (not (Variables|Hidden|GetEditMode(Don'tTouch)))
    (Utilities|Array|Clear (Variables|Default|GetNewlySpawnedActors))
    (for _array_element (Variables|Hidden|GetPropdata)
      (bind (_array_element_actor_27_6a42496a4663e346c97afeb9993c68f7 _array_element_transform_11_ce6eb422421fe49e3943bdb094b64c74 _array_element_progressappear_21_5b07da4e4ffa7913e1c2ef864c0bdc68 _array_element_progressdisappear_22_d8fc05f84ecb57e73362de9563548198 _array_element_parentsmactor_26_e003ce7b46936c6cc054b9af925c1bde) (Utilities|Struct|BreakCalendarPropData _array_element))
      (bind _returnvalue (Utilities|LoadClassAssetBlocking _array_element_actor_27_6a42496a4663e346c97afeb9993c68f7))
      (bind _asactor (Utilities|Casting|CastToActorClass _returnvalue))
      (bind _returnvalue_2 (Game|SpawnActorfromClass _asactor (Math|Transform|ComposeTransforms _array_element_transform_11_ce6eb422421fe49e3943bdb094b64c74 (Transformation|GetActorTransform)) "AlwaysSpawn" "SelectDefaultAtRuntime" 0 0))
      (Utilities|Array|Add (Variables|Default|GetActors) _returnvalue_2)
      (Utilities|Array|Add (Variables|Default|GetParentToAttach) _array_element_parentsmactor_26_e003ce7b46936c6cc054b9af925c1bde)
      (Utilities|Array|Add (Variables|Default|GetNewlySpawnedActors) _returnvalue_2))
    (for (_array_element_2 _array_index) (Variables|Default|GetNewlySpawnedActors)
      (bind _asbuilding_prop (Utilities|Casting|CastToBuildingProp _array_element_2))
      (Utilities|IsValid _asbuilding_prop
        (:Is%20Valid
          (if (> (.ParentArrayElement_30_0DF457554DE47CFC3CCE8DA1B253825E (Utilities|Struct|BreakCalendarPropData (Utilities|Array|Get(acopy) _propdata _array_index))) -1)
            (bind _asbuilding_prop_2 (Utilities|Casting|CastToBuildingProp (Utilities|Array|Get(acopy) (Variables|Default|GetNewlySpawnedActors) (.ParentArrayElement_30_0DF457554DE47CFC3CCE8DA1B253825E (Utilities|Struct|BreakCalendarPropData (Utilities|Array|Get(acopy) _propdata _array_index)))))) ; Parent actor taken from spawned array
            (Attachement|SetParentActortoAttachTo _asbuilding_prop _asbuilding_prop_2)
            (else
              (Attachement|SetParentActortoAttachTo _asbuilding_prop (.ParentSMactor_26_E003CE7B46936C6CC054B9AF925C1BDE (Utilities|Struct|BreakCalendarPropData (Utilities|Array|Get(acopy) _propdata _array_index)))))))))
    (Variables|Hidden|SetEditMode(Don'tTouch) true)))

(event Collision|EventActorBeginOverlap
  (params (OtherActor :type Actor%20Object%20Reference)))

(event EventTick
  (params (DeltaSeconds :type Float%20%28single-precision%29)))

(fn GetArrayElementAttachedTo
  (params (SpawnedActor :type Actor%20Object%20Reference))
  (bind _asbuilding_static_mesh (Utilities|Casting|CastToBuildingSMActor SpawnedActor))
  (for (_array_element _array_index) (Variables|Default|GetActors)
    (if (Utilities|Equal(Object) _array_element (Variables|/Script/FortniteGame.BuildingSMActor|GetParentActorToAttachTo _asbuilding_static_mesh))
      (Development|PrintString (Utilities|String|Append (Utilities|GetDisplayName _asbuilding_static_mesh) " FOUND PROP TO ATTACH TO: " (Utilities|GetDisplayName _array_element)) true true "(R=0.000000,G=0.660000,B=1.000000,A=1.000000)" 15.0)
      (return true _array_index)))
  (return false -1))

(fn OnRep_GrowthStage
  (CallFunction|SpawnActors))

(fn ConstructionScript)
