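; BP_DataLayerRandom is a prototype Blueprint that picks a random World Partition data layer setup when play begins, on authority only.
; If its list of setups is non-empty, it shuffles the list and applies the first entry, or applies a chosen index when the debug layer spawn flag is on.
; ApplyDataLayerSetup activates that entry's enabled data layers and unloads its disabled ones through the Data Layer Manager, non-recursively.
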(event EventBeginPlay
  (Networking|SwitchHasAuthority
    (:Authority
      (if (> (Utilities|Array|Length (Variables|Default|GetDataLayersList)) 0)
        (if (Variables|Default|GetDebugEnableLayerSpawn)
          (CallFunction|ApplyDataLayerSetup (Utilities|Array|Get(acopy) (Variables|Default|GetDataLayersList) (Variables|Default|GetDebugLayerSpawnIndex)))
          (else
            (Utilities|Array|Shuffle (Variables|Default|GetDataLayersList))
            (CallFunction|ApplyDataLayerSetup (Utilities|Array|Get(acopy) (Variables|Default|GetDataLayersList)))))))))

(fn ApplyDataLayerSetup
  (params (RandomDataLayerData :type Random%20Data%20Layer%20Data%20Structure%20%28by%20ref%29))
  (bind (_enableddatalayers_9_41bb9a3e4648805b5af70aa728288c92 _disableddatalayers_10_8de59e3b474c88ed2e7988b185572444) (Utilities|Struct|BreakRandomDataLayerData RandomDataLayerData))
  (bind _returnvalue (WorldPartition|GetDataLayerManager))
  (Utilities|FlowControl|Sequence
    (:then_0
      (for _array_element _enableddatalayers_9_41bb9a3e4648805b5af70aa728288c92
        (DataLayers|SetDataLayerInstanceRuntimeState _returnvalue (DataLayers|GetDataLayerInstancefromAsset _returnvalue _array_element) "Activated" false)))
    (:then_1
      (for _array_element_2 _disableddatalayers_10_8de59e3b474c88ed2e7988b185572444
        (DataLayers|SetDataLayerInstanceRuntimeState _returnvalue (DataLayers|GetDataLayerInstancefromAsset _returnvalue _array_element_2) "Unloaded" false)))))

(fn ConstructionScript)
