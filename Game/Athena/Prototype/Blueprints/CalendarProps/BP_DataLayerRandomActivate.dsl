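; Prototype calendar prop actor that picks a random data layer setup at BeginPlay on the authority, gated by the Turtle Island hotfix flag.
; It filters setups by their enabled flags, then shuffles and takes the first, unless a positive hotfix index selects a 1-based entry.
; A debug flag applies every enabled setup instead.
; Each applied setup activates and unloads its listed World Partition data layers.
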(event EventBeginPlay
  (bind _hf_indexoverride (Variables|Hotfix|GetHF_IndexOverride))
  (bind _returnvalue (ScalableFloat|GetValueasInteger _hf_indexoverride))
  (bind _returnvalue_2 (<= _returnvalue 0))
  (Networking|SwitchHasAuthority
    (:Authority
      (if (ScalableFloat|GetValueasBool (Variables|Hotfix|GetHF_TurtleIslandEnabled))
        (bind _enableddatalayers (CallFunction|GetAllEnabledDataLayers (Variables|Default|GetDataLayerList)))
        (Variables|Default|SetEnabledLayers _enableddatalayers)
        (if (> (Utilities|Array|Length (Variables|Default|GetEnabledLayers)) 0)
          (if _returnvalue_2
            (Utilities|Array|Shuffle (Variables|Default|GetEnabledLayers))
            (else
              _))
          (Utilities|Array|AppendArray (Variables|Default|GetDataLayerSetupsToApply) (select (ScalableFloat|GetValueasBool (Variables|Debug|GetShouldEnableAllPossibleLocations)) (Variables|Default|GetEnabledLayers) (Utilities|Array|MakeArray (Utilities|Array|Get(acopy) (Variables|Default|GetEnabledLayers) (Math|Integer|SelectInt 0 (- _returnvalue 1) _returnvalue_2)))))
          (for _array_element (Variables|Default|GetDataLayerSetupsToApply)
            (CallFunction|ApplyDataLayerSetup _array_element)))))))

(fn ApplyDataLayerSetup
  (params (RandomDataLayerData :type Random%20Data%20Layer%20Data%20Structure%20%28by%20ref%29))
  (bind (_enableddatalayers_9_41bb9a3e4648805b5af70aa728288c92 _disableddatalayers_10_8de59e3b474c88ed2e7988b185572444) (Utilities|Struct|BreakRandomDataLayerData RandomDataLayerData))
  (bind _returnvalue (WorldPartition|GetDataLayerManager))
  (Utilities|FlowControl|Sequence
    (:then_0
      (for _array_element _enableddatalayers_9_41bb9a3e4648805b5af70aa728288c92
        (DataLayers|SetDataLayerInstanceRuntimeState _returnvalue (DataLayers|GetDataLayerInstancefromAsset _returnvalue _array_element) "Activated" false)))
    (:then_1
      (for _array_element_2 _disableddatalayers_10_8de59e3b474c88ed2e7988b185572444
        (DataLayers|SetDataLayerInstanceRuntimeState _returnvalue (DataLayers|GetDataLayerInstancefromAsset _returnvalue _array_element_2) "Unloaded" false)))))

(fn GetAllEnabledDataLayers
  (params (DataLayersToCheck :type Array%20of%20Random%20Data%20Layer%20Data%20Structures))
  (for _array_element DataLayersToCheck
    (if (ScalableFloat|GetValueasBool (.bSetupEnabled_22_436E2FE243B567AE5E7CA9AAC2346CA0 (Utilities|Struct|BreakRandomDataLayerData _array_element)))
      (Utilities|Array|Add (LocalVariables|GetFilteredLayers) _array_element)))
  (return (LocalVariables|GetFilteredLayers)))

(fn ConstructionScript)
