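; Calendar prop prototype Blueprint that activates one randomly chosen World Partition data layer when play begins.
; On authority only, BeginPlay shuffles a non-empty DataLayersToPickFrom array and sets the first entry's data layer instance to Activated through the DataLayerManager.
; A helper, GetAllEnabledDataLayers, collects entries whose ScalableFloat setup flag is true into FilteredLayers, but the BeginPlay path does not call it.
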(event EventBeginPlay
  (Networking|SwitchHasAuthority
    (:Authority
      (if (> (Utilities|Array|Length (Variables|Default|GetDataLayersToPickFrom)) 0)
        (Utilities|Array|Shuffle (Variables|Default|GetDataLayersToPickFrom))
        (bind _returnvalue (WorldPartition|GetDataLayerManager))
        (DataLayers|SetDataLayerInstanceRuntimeState _returnvalue (DataLayers|GetDataLayerInstancefromAsset _returnvalue (Utilities|Array|Get(acopy) (Variables|Default|GetDataLayersToPickFrom))) "Activated" false)))))

(fn GetAllEnabledDataLayers
  (params (DataLayersToCheck :type Array%20of%20Random%20Data%20Layer%20Data%20Structures) (DataLayers :type Array%20of%20Data%20Layer%20Asset%20Object%20References))
  (for _array_element DataLayersToCheck
    (if (ScalableFloat|GetValueasBool (.bSetupEnabled_22_436E2FE243B567AE5E7CA9AAC2346CA0 (Utilities|Struct|BreakRandomDataLayerData _array_element)))
      (Utilities|Array|Add (Variables|Default|GetFilteredLayers) _array_element)))
  (return (Variables|Default|GetFilteredLayers)))

(fn ConstructionScript)
