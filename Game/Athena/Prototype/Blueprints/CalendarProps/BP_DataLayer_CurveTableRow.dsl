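; Calendar prop Blueprint actor that switches a list of World Partition data layers on or off based on a curve table setting.
; On BeginPlay it waits for playlist data to be ready, then runs only with authority.
; If the scalable float row is valid and true, and the playlist context tags match none of the excluded tags, every listed layer is set Activated; otherwise each is set Unloaded.
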(event EventBeginPlay
  (Class|FortAsyncActionCurrentPlaylistReady|PlaylistDataReadyAsync
    (:OnReady
      (Networking|SwitchHasAuthority
        (:Authority
          (bind _returnvalue (WorldPartition|GetDataLayerManager))
          (if (Math|Boolean|ANDBoolean (ScalableFloat|IsSettingCurveValid (Variables|Default|GetCurveTableRow)) (ScalableFloat|GetValueasBool (Variables|Default|GetCurveTableRow)) (not (GameplayTags|HasAnyTags (Class|FortGameStateAthena|GetAthenaPlaylistContextTags (Fortnite|Athena|GameState|GetGameStateAthena)) (Variables|Default|GetExcludePlaylists_With_Tags))))
            (for _array_element (Variables|Default|GetDataLayerAssetList)
              (DataLayers|SetDataLayerInstanceRuntimeState _returnvalue (DataLayers|GetDataLayerInstancefromAsset _returnvalue _array_element) "Activated" false) ; Enable/Activate all listed Data Layers
            )
            (else
              (for _array_element_2 (Variables|Default|GetDataLayerAssetList)
                (DataLayers|SetDataLayerInstanceRuntimeState _returnvalue (DataLayers|GetDataLayerInstancefromAsset _returnvalue _array_element_2) "Unloaded" false) ; Disable all listed Data Layers
              ))))))))

(fn ConstructionScript)
