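; Calendar-driven prop Blueprint (BP_SynthContact_CalendarDataLayer) that switches World Partition data layers when named calendar events start or end.
; GetEventNames reports unique, non-empty names from CalendarDataLayerList, and both event handlers feed HandleEventStateChanged.
; For matching entries, it sets Enabled and Disabled layers to opposite Activated/Unloaded states, inverted when the event ends, with verbose logging.
; CanUpdateInMatch refuses mid-match updates for performance reasons, logging the attempt when the parent would allow one.
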(event EventBeginPlay)

(event Collision|EventActorBeginOverlap
  (params (OtherActor :type Actor%20Object%20Reference)))

(event EventTick
  (params (DeltaSeconds :type Float%20%28single-precision%29)))

(fn CanUpdateInMatch
  (if (|CallParentFunction)
    (Class|BuildingFunctionLibrary|LogString "NewEnumerator5" "Attempting to update a Data Layer Calendar Event during a match.  This is not supported due to performance concerns." true self)
    (return false)))

(fn GetEventNames
  (for _array_element (Variables|Default|GetCalendarDataLayerList)
    (if (not (Utilities|String|IsEmpty (Utilities|Struct|BreakCalendarDataLayerData _array_element)))
      (Utilities|Array|AddUnique (LocalVariables|GetLocalEventNames) (Utilities|Struct|BreakCalendarDataLayerData _array_element))))
  (return (LocalVariables|GetLocalEventNames)))

(fn HandleEventActive
  (params (EventName :type String))
  (CallFunction|HandleEventStateChanged EventName true))

(fn HandleEventEnded
  (params (EventName :type String))
  (CallFunction|HandleEventStateChanged EventName))

(fn HandleEventStateChanged
  (params (EventName :type String) (bIsEventActive :type Boolean))
  (bind _biseventactive (Variables|GetbIsEventActive))
  (bind _returnvalue (select _biseventactive "Unloaded" "Activated"))
  (bind _biseventactive_2 (Variables|GetbIsEventActive))
  (bind _returnvalue_2 (select _biseventactive_2 "Activated" "Unloaded"))
  (bind _eventname (Variables|GetEventName))
  (for _array_element (Variables|Default|GetCalendarDataLayerList)
    (if (Utilities|String|EqualExactly(String) (Variables|GetEventName) (Utilities|Struct|BreakCalendarDataLayerData _array_element))
      (CallFunction|LogIt (Utilities|String|ToString(Text) (Utilities|Text|FormatText "NSLOCTEXT(\"[A7AEC95803B6180A6B97DD22DA234F27]\", \"9A79C6704BB570CA0078828AA4475376\", \"{a} : BP_CalendarDataLayer:HandleEventActive - Calendar Event Name: {b} was found.\")" (Variables|Default|GetDisplayName) _eventname)))
      (bind _returnvalue_3 (WorldPartition|GetDataLayerManager))
      (CallFunction|LogIt (Utilities|String|ToString(Text) (Utilities|Text|FormatText "NSLOCTEXT(\"[A7AEC95803B6180A6B97DD22DA234F27]\", \"57D104B94FEF3170F132D9B24C615C7E\", \"{a} : BP_CalendarDataLayer:HandleEventActive - Getting Data Layer Manager : {b}\")" (Variables|Default|GetDisplayName) _returnvalue_3)))
      (Utilities|FlowControl|Sequence
        (:then_0
          (for _array_element_2 (.EnabledDataLayers_9_41BB9A3E4648805B5AF70AA728288C92 (Utilities|Struct|BreakCalendarDataLayerData _array_element))
            (DataLayers|SetDataLayerInstanceRuntimeState _returnvalue_3 (DataLayers|GetDataLayerInstancefromAsset _returnvalue_3 _array_element_2) _returnvalue_2 false)
            (CallFunction|LogIt (Utilities|String|ToString(Text) (Utilities|Text|FormatText "NSLOCTEXT(\"[A7AEC95803B6180A6B97DD22DA234F27]\", \"0D231ED34BCA49EAE50CBE9B5A2593D9\", \"{a} : BP_CalendarDataLayer:HandleEventActive:ENABLED Layers  from the Enable Layer- Calling Set Data Layer Instance Runtime State with the following information:\\r\\nDataLayer Manager: {b}\\r\\nDataLayer Instance: {c}\\r\\nSetting to State: {d}\\r\\nBecause Event is: {e}\")" (Variables|Default|GetDisplayName) _returnvalue_3 (DataLayers|GetDataLayerInstancefromAsset _returnvalue_3 _array_element_2) (Utilities|String|EnumtoString _returnvalue_2) (Variables|GetbIsEventActive))))))
        (:then_1
          (for _array_element_3 (.DisabledDataLayers_10_8DE59E3B474C88ED2E7988B185572444 (Utilities|Struct|BreakCalendarDataLayerData _array_element))
            (DataLayers|SetDataLayerInstanceRuntimeState _returnvalue_3 (DataLayers|GetDataLayerInstancefromAsset _returnvalue_3 _array_element_3) _returnvalue false)
            (CallFunction|LogIt (Utilities|String|ToString(Text) (Utilities|Text|FormatText "NSLOCTEXT(\"[A7AEC95803B6180A6B97DD22DA234F27]\", \"DC56B8834ACF8648266E4FB9F87D8A5D\", \"{a} : BP_CalendarDataLayer:HandleEventActive:DISABLED Layers from the Disable Array - Calling Set Data Layer Instance Runtime State with the following information:\\r\\nDataLayer Manager: {b}\\r\\nDataLayer Instance: {c}\\r\\nSetting to State: {d}\\r\\nBecause Event is: {e}\")" (Variables|Default|GetDisplayName) _returnvalue_3 (DataLayers|GetDataLayerInstancefromAsset _returnvalue_3 _array_element_3) (Utilities|String|EnumtoString _returnvalue) (Variables|GetbIsEventActive)))))))
      (else
        (CallFunction|LogIt (Utilities|String|ToString(Text) (Utilities|Text|FormatText "NSLOCTEXT(\"[A7AEC95803B6180A6B97DD22DA234F27]\", \"58607B014BC58DD1C34B6A9DB27A5C95\", \"{a} : BP_CalendarDataLayer:HandleEventActive - Calendar Event Name: {b} was NOT found.\")" (Variables|Default|GetDisplayName) _eventname)) true)))))

(fn ConstructionScript
  (|CallParentFunction))
